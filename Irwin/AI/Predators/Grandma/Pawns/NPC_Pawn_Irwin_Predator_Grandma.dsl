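; Pawn Blueprint for the Grandma predator NPC in the Irwin wildlife AI, layering effects and interaction rules over its parent.
; Landing reuses or spawns an attached Niagara effect, flags mud by physical surface, and is disabled inside water bodies.
; Interaction is gated by gameplay tags and blocked at certain alert levels while another pawn is converting it.
; It also attaches idle-eye FX, limits auto-fire to hostile attackers, and randomizes hit-react direction.
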(event EventOnLanded
  (params (Hit :type Hit%20Result%20Structure%20%28by%20ref%29))
  (bind (_bblockinghit _binitialoverlap _time _distance _location _impactpoint _normal _impactnormal _physmat) (Collision|BreakHitResult Hit))
  (bind _landfx_comp (Variables|Default|GetLandFX_Comp))
  (|CallParentFunction Hit)
  (if (Variables|Default|GetplayLandingFX)
    (bind _landfx_comp_2 (Variables|Default|GetLandFX_Comp)
      (:then
        (Components|Activation|Activate _landfx_comp_2 true))
      (:else
        (bind _returnvalue (Niagara|SpawnSystemAttached (Variables|Default|GetLanding%20FX) (Variables|Character|GetMesh) "Root" "0, 0, 0" "0, 0, 0" "KeepRelativeOffset" true))
        (Variables|Default|SetLandFX_Comp _returnvalue)))
    (if (Utilities|Enum|Equal(Enum) (Variables|/Script/PhysicsCore.PhysicalMaterial|GetSurfaceType _physmat) "SurfaceType30")
      (Effects|Components|ParticleSystem|SetBoolParameter _landfx_comp "IsOnMud" true)
      (else
        (Effects|Components|ParticleSystem|SetBoolParameter _landfx_comp "IsOnMud" false)))))

(event WaterBody|EventOnEnteredWaterBody
  (params (WaterBody :type Fort%20Water%20Body%20Actor%20Object%20Reference) (bIsFirstBody :type Boolean))
  (|CallParentFunction WaterBody bIsFirstBody)
  (Variables|Default|SetplayLandingFX false))

(event WaterBody|EventOnExitedWaterBody
  (params (WaterBody :type Fort%20Water%20Body%20Actor%20Object%20Reference) (bIsLastBody :type Boolean))
  (|CallParentFunction WaterBody bIsLastBody)
  (Variables|Default|SetplayLandingFX true))

(fn ApplyVariation
  (|CallParentFunction)
  (Transformation|AttachComponentToComponent (Variables|Default|GetNS_Irwin_Grandma_IdleEye) (Variables|Transformation|GetRootComponent) "Root_Effect" "KeepRelative" "KeepRelative" "KeepRelative" false))

(fn BlueprintCanInteract
  (params (InteractingPawn :type Fort%20Pawn%20Object%20Reference) (InteractionBeingAttempted :type EInteractionBeingAttempted%20Enum :default "FirstInteraction"))
  (return (Math|Boolean|ANDBoolean (|CallParentFunction InteractingPawn InteractionBeingAttempted) (not (GameplayTags|HasAnyMatchingGameplayTags InteractingPawn (Variables|Default|GetInteractingPawnPreventingTags))) (select (GameplayTags|HasAnyMatchingGameplayTags (LocalVariables|GetCreative%20Taming%20Allowed%20Tags)) true (not (GameplayTags|HasAnyMatchingGameplayTags (Variables|Default|GetInteraction%20Preventing%20Tags)))) (not (Math|Boolean|ANDBoolean (Utilities|IsValid (Convert|GetConverterPawn (Variables|Default|GetConvertComponent)) :!class /Script/Engine.KismetSystemLibrary) (!= (Convert|GetConverterPawn (Variables|Default|GetConvertComponent)) (Variables|GetInteractingPawn)) (or (Utilities|Enum|Equal(Enum) (Variables|NPCAlertLevel|GetNPC_AlertLevel) "NewEnumerator2") (Utilities|Enum|Equal(Enum) (Variables|NPCAlertLevel|GetNPC_AlertLevel) "NewEnumerator3")))))))

(fn IsValidAutoFireTarget
  (params (Attacker :type Actor%20Object%20Reference))
  (return (Utilities|Enum|Equal(Enum) (Team|GetTeamAffiliation Attacker self) "Hostile")))

(fn PlayAdditiveHitReacts
  (params (Hit%20Direction :type Vector%20%28by%20ref%29 :default "0, 0, 0") (Anim%20Montage :type Anim%20Montage%20Object%20Reference))
  (|CallParentFunction (Math|Random|RandomUnitVector) Anim%20Montage)
  (return))

(fn ConstructionScript
  (|CallParentFunction)
  ;[1 1 1] Delete after s22
  (Class|BTechArtGlobalLibrary|AddActorTag self (Utilities|Name|MakeLiteralName "Lackluster.Material.AI.Grandma")))
